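# Idle cosmetic Niagara effect for the ForestJeanWink character, emitting audio-reactive star sprites around the body.
# One packed emitter runs four spawn groups into five sprite renderers, placing particles at skeletal mesh bones with hand-authored offset arrays.
# A fake audio loop drives spawn rate and amplitude-based size, speed and brightness, alongside variant tint and time-of-day scaling.
# Hidden head/hat, hand and foot flags kill matching particles and rebalance spawn rates; velocity fade dims and culls particles.
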NiagaraSystem /BRCosmetics/Characters/Player/Female/Medium/Bodies/F_MED_ForestJeanWink/FX/NS_Character_ForestJeanWink_Idle
dependencies {
 usedDataInterfaces [
  &NiagaraDataInterfaceArrayFloat3
  &NiagaraDataInterfaceCurve
  &NiagaraDataInterfaceVector2DCurve
  &NiagaraDataInterfaceSkeletalMesh
  &NiagaraDataInterfaceColorCurve
 ]
 usedDynamicInputs [
  &RandomRangeLinearColor
  &ModuloFloat
  &FN_FrontEndFloatModifier
  /Game/Effects/Niagara/DynamicInputs/Math/FN_InterpFloat
  &VectorLength
  &NormalizeVector
  &MakeCustomFloatFromBool
  /Niagara/DynamicInputs/Lerp/SmoothLerpOverTime_Float
  &RemapRange
  &MakeFloatFromLinearColor
  &MakeIntFromFloat
  &ColorFromCurve
  &FNTimeOfDayScalar
 ]
 usedModules [
  &FN_VariantManager
  &self:SystemSpawnScript.SystemScriptSource.SystemScriptGraph.NiagaraNodeAssignment_3.SetVariables_#1
  &self:SystemSpawnScript.SystemScriptSource.SystemScriptGraph.NiagaraNodeAssignment_4.SetVariables_#2
  &self:SystemSpawnScript.SystemScriptSource.SystemScriptGraph.NiagaraNodeAssignment_5.SetVariables_#3
  &self:SystemSpawnScript.SystemScriptSource.SystemScriptGraph.NiagaraNodeAssignment_7.SetVariables_#4
  &SystemState
  &self:SystemSpawnScript.SystemScriptSource.SystemScriptGraph.NiagaraNodeAssignment_9.SetVariables_#5
  &Timeline
  /Game/Effects/Niagara/Modules/Audio/FN_FakeAudioReactivityLoop
  &self:SystemSpawnScript.SystemScriptSource.SystemScriptGraph.NiagaraNodeAssignment_2.SetVariables_#6
  &self:SystemSpawnScript.SystemScriptSource.SystemScriptGraph.NiagaraNodeAssignment_6.SetVariables_#7
  &self:SystemSpawnScript.SystemScriptSource.SystemScriptGraph.NiagaraNodeAssignment_1.SetVariables_#8
  &self:SystemSpawnScript.SystemScriptSource.SystemScriptGraph.NiagaraNodeAssignment_0.SetVariables_#9
  &self:SystemSpawnScript.SystemScriptSource.SystemScriptGraph.NiagaraNodeAssignment_8.SetVariables_#10
  &self:SystemSpawnScript.SystemScriptSource.SystemScriptGraph.NiagaraNodeAssignment_10.SetVariables_#11
  &self:SystemSpawnScript.SystemScriptSource.SystemScriptGraph.NiagaraNodeAssignment_12.SetVariables_#12
 ]
}
emitters {Stars_Packed {
  emitterSpawnScript [/SetVariables(Emitter.NumberOfRegion=4) {isSetParameters=true}]
  emitterUpdateScript [
   &EmitterState(Inactive Response=::NewEnumerator1, Life Cycle Mode=::NewEnumerator1, Loop Duration=1, UseLoopDelay=false)
   /SetVariables(Emitter.NumberOfEnabledRegion:NiagaraInt32=&Add_Integer(
     A=1
     B=&Add_Integer(
      A=&MakeCustomIntFromBool(Bool=&Boolean_OrOperation(Bool1=$User.bisHeadHidden, Bool2=$System.bisHatHidden), True INT=0, False INT=1)
      B=&Add_Integer(
       A=&MakeCustomIntFromBool(Bool=$User.bisHandHidden, True INT=0, False INT=1)
       B=&MakeCustomIntFromBool(Bool=$User.bisFootHidden, True INT=0, False INT=1)
      )
     )
    )) {isSetParameters=true}
   &SpawnRate(
    SpawnRate=&Lerp_Float(
     A=0
     B=&Multiply_Float(
      A=60
      B=&MakeFloatFromInt(Int=/Niagara/DynamicInputs/Divide/Divide_Int(A:NiagaraInt32=$Emitter.NumberOfRegion, B:NiagaraInt32=$Emitter.NumberOfEnabledRegion))
     )
     Alpha=$System.AudioReactivity
    )
    Use Spawn Probability=false
   )
   &SpawnRate(SpawnGroup=1, SpawnRate=&Lerp_Float(A=0, B=5, Alpha=$System.AudioReactivity), Use Spawn Probability=false) {module=SpawnRate001}
   &SpawnRate(
    SpawnGroup=2
    SpawnRate=&Lerp_Float(
     A=0
     B=&Multiply_Float(
      A=18
      B=&MakeFloatFromInt(Int=&Divide_Int(A:NiagaraInt32=$Emitter.NumberOfRegion, B:NiagaraInt32=$Emitter.NumberOfEnabledRegion))
     )
     Alpha=$System.AudioReactivity
    )
    Use Spawn Probability=false
   ) {module=SpawnRate002}
   &SpawnRate(
    SpawnGroup=3
    SpawnRate=&Lerp_Float(
     A=0
     B=&Multiply_Float(
      A=10
      B=&MakeFloatFromInt(Int=&Divide_Int(A:NiagaraInt32=$Emitter.NumberOfRegion, B:NiagaraInt32=$Emitter.NumberOfEnabledRegion))
     )
     Alpha=$System.AudioReactivity
    )
    Use Spawn Probability=false
   ) {module=SpawnRate003}
  ]
  particleSpawnScript [
   &FN_SetSpawnGroup(Number of Spawn Groups:NiagaraInt32=&MergeHelper_IntSwitch(0=6, 1=2, 2=6, 3=6, 4=?, 5=?, Index=$Emitter.SpawnGroup, Num=::NewEnumerator2))
   &PartitionParticles(Partition Mode:ENiagara_PartitionMode_v2=::NewEnumerator5)
   &InitializeParticle(
    Color Mode=::NewEnumerator0
    Lifetime=&MergeHelper_FloatSwitch(
     0=&RandomRangeFloat(Minimum=0.35, Maximum=0.5)
     1=&RandomRangeFloat(Minimum=0.18, Maximum=0.225)
     2=&RandomRangeFloat(Minimum=0.375, Maximum=0.4875)
     3=&RandomRangeFloat(Minimum=0.27, Maximum=0.3375)
     Index=$Emitter.SpawnGroup
     Num=::NewEnumerator2
    )
    Sprite Rotation Angle=&MergeHelper_FloatSwitch(
     0=&RandomRangeFloat(Minimum=0, Maximum=1)
     1=&RandomRangeFloat(Minimum=0, Maximum=0.2)
     2=0
     3=&RandomRangeFloat(Minimum=0, Maximum=0.2)
     Index=$Emitter.SpawnGroup
     Num=::NewEnumerator2
    )
    Sprite Rotation Mode=::NewEnumerator1
    Sprite Size Mode=::NewEnumerator3
    Sprite UV Mode=::NewEnumerator3
    Uniform Sprite Size=&RandomRangeFloat(Minimum=3, Maximum=5.6667)
    UsePositionOffset=false
    Write Material Random=true
   )
   &CameraOffset(Camera Offset Amount=&MergeHelper_FloatSwitch(0=12, 1=25, 2=12, 3=12, Index=$Emitter.SpawnGroup, Num=::NewEnumerator2))
   /SetVariables(Particles.SampleFront:NiagaraFloat=&MergeHelper_FloatSwitch(
     0=&Lerp_Float(A=0, B=1, Alpha=&MakeFloatFromInt(Int=$Particles.PartitionParticles.Partition))
     1=0
     2=&Lerp_Float(A=0, B=1, Alpha=&MakeFloatFromInt(Int=$Particles.PartitionParticles.Partition))
     3=&Lerp_Float(A=0, B=1, Alpha=&MakeFloatFromInt(Int=$Particles.PartitionParticles.Partition))
     4=0
     Index=$Emitter.SpawnGroup
     Num=::NewEnumerator3
    )) {isSetParameters=true}
   /SetVariables(Particles.RotationDirection:NiagaraFloat=&UniformAOrBFloat(A=-1, Evaluation Type=::NewEnumerator0, Recalculate Random Each Loop=true, Override Seed=false)) {isSetParameters=true}
   /SetVariables(Particles.VisibilityTag=$Emitter.SpawnGroup) {isSetParameters=true}
   &SkeletalMeshLocation(
    Bone Sampling Mode=::NewEnumerator4
    MaskSpawnGroup=false
    Orientation Sampling=::NewEnumerator1
    Position Sampling=::NewEnumerator2
    Skeletal Mesh=$System.Skeletal Mesh
   ) {module=SkeletalMeshLocation001}
   &SkeletalMeshLocation(
    Bone / Socket Index=&MergeHelper_IntSwitch(
     0=&SpawnGroup()
     1=&Add_Integer(A=6, B=&SpawnGroup())
     2=&SpawnGroup()
     3=&SpawnGroup()
     4=?
     5=?
     Index=$Particles.VisibilityTag
     Num=::NewEnumerator2
    )
    Bone Sampling Mode=::NewEnumerator4
    MaskSpawnGroup=false
    Sampled Position Offset=&MergeHelper_VectorSwitch(
     0=&MaskVector(
      Mask Source=1
      Mask Parameter=$Particles.VisibilityTag
      Mask=1
      Masked Vector=&SelectVectorFromArray(
       Array Sampling Mode=::NewEnumerator0
       Direct Array Index=?
       Vector Selection Array=DataInterface() {propertyValues {FloatData [
          <-50.4914 -2.1142 4.3092>
          <-34.7061 2.6521 5.6199>
          <-36.6385 6.6296 -0.7165>
          <-21.5877 6.4904 3.8227>
          <-19.0434 6.0873 4.2198>
          <-16.2989 -4.4494 7.8867>
          <-30.7479 -6.5444 3.6111>
          <-25.3201 -8.2397 1.6508>
          <-46.8999 3.4244 1.0709>
          <-43.7631 4.1476 -2.2229>
          <-42.523 -3.1488 -3.9349>
          <-26.325 1.8001 -7.2802>
          <-25.8992 6.1614 -5.1872>
          <-20.7565 7.2412 -4.3787>
          <-14.6051 4.4848 -7.3844>
         ]}}
       Random Seed=0
      )
      Passthrough Vector=&Lerp_Vector(
       A=&SelectVectorFromArray(
        Array Sampling Mode=::NewEnumerator0
        Direct Array Index=?
        Vector Selection Array=DataInterface() {propertyValues {FloatData [
           <-16.5872 -14.3335 -2.871>
           <-18.7521 -13.8851 0.5084>
           <-11.4786 -12.8514 -4.5738>
           <-7.5487 -7.9109 -10.0271>
           <-5.9966 -8.573 -6.5224>
           <-6.482 -9.2692 -4.3985>
           <-9.6908 -11.7174 -3.8704>
           <-9.1888 -11.2498 -2.2197>
           <-12.0146 -12.8871 1.3266>
           <-18.2386 -14.6251 4.0794>
           <-12.3165 -13.1903 3.2307>
           <-4.4084 -7.3638 6.574>
           <-7.6605 -10.3377 0.0351>
           <-2.4717 -7.164 1.9928>
           <-1.6521 -6.8863 -0.2809>
           <0.3251 -7.1909 2.269>
           <-19.8807 -14.2886 4.4657>
           <-8.5186 -7.1855 11.4681>
           <-14.22 -9.235 13.4966>
           <-19.452 -13.6669 10.3588>
           <-19.2324 -14.2352 9.03>
           <-15.6806 -9.0032 -14.1858>
           <-20.7064 -14.4016 -5.8932>
           <-17.0158 -14.5132 -5.746>
           <-13.877 -12.4 -10.1413>
          ]}}
        Random Seed=0
       )
       B=&SelectVectorFromArray(
        Array Sampling Mode=::NewEnumerator0
        Direct Array Index=?
        Vector Selection Array=DataInterface() {propertyValues {FloatData [
           <0.8726 3.337 -9.1225>
           <3.0414 9.8472 0.7452>
           <11.9054 11.7926 -0.1974>
           <-0.4028 9.0282 -3.1445>
           <-6.1549 1.3445 12.069>
           <-11.9034 -1.2344 14.7566>
           <-14.1035 -5.2432 -15.1765>
           <-11.2577 3.1003 -13.888>
           <-3.8803 0.0381 -10.8421>
           <8.8792 -6.3778 -9.3257>
           <-2.4401 -3.6553 -9.1801>
           <10.2071 0.7741 12.125>
           <-2.4043 6.0444 8.0867>
           <2.2118 3.3787 9.2786>
           <8.194 10.9602 3.7746>
           <-3.0019 8.403 4.5695>
           <10.7687 11.7222 -1.4627>
           <-2.3812 7.6103 -6.1175>
           <-12.9737 2.1263 14.6695>
           <-14.7763 -5.8594 15.2005>
          ]}}
        Random Seed=0
       )
       Alpha=$Particles.SampleFront
      )
     )
     1=&MaskVector(
      Mask Source=1
      Mask Parameter=$Particles.VisibilityTag
      Mask=1
      Masked Vector=&SelectVectorFromArray(
       Array Sampling Mode=::NewEnumerator0
       Direct Array Index=?
       Vector Selection Array=DataInterface() {propertyValues {FloatData [
          <50.454 7.4613 4.1795>
          <37.6372 -6.1517 -1.7253>
          <44.3233 0.1383 4.4669>
          <34.0083 1.1553 -6.2015>
          <23.1141 -6.4103 -3.9094>
          <32.1924 -7.5154 1.3636>
          <40.9152 -0.7685 4.4645>
          <26.9144 8.011 -1.0488>
          <18.6609 2.0473 8.2236>
          <17.5058 8.7377 2.5381>
          <11.2761 7.4238 5.8469>
          <13.8399 -7.7709 1.6664>
          <17.0145 9.249 -3.0128>
          <17.4299 8.4545 -4.6992>
         ]}}
       Random Seed=0
      )
      Passthrough Vector=&Lerp_Vector(
       A=&SelectVectorFromArray(
        Array Sampling Mode=::NewEnumerator0
        Direct Array Index=?
        Vector Selection Array=DataInterface() {propertyValues {FloatData [
           <21.9958 -3.2608 1.2913>
           <10.8478 -2.7695 -3.1038>
           <18.7528 -3.664 -0.2212>
           <11.045 -3.7815 1.5256>
           <2.6529 -2.317 3.0553>
           <12.9969 -3.7737 -1.3173>
           <15.5958 -3.621 -0.9829>
          ]}}
        Random Seed=0
       )
       B=&SelectVectorFromArray(
        Array Sampling Mode=::NewEnumerator0
        Direct Array Index=?
        Vector Selection Array=DataInterface() {propertyValues {FloatData [
           <22.5453 1.978 1.1021>
           <17.4327 1.0292 2.4129>
           <17.7429 1.5517 -2.1364>
           <5.0588 1.2863 4.5403>
           <7.0503 2.9234 3.5792>
           <10.392 3.1425 -0.8809>
           <9.638 3.6669 0.2739>
           <11.4049 0.0728 -3.3303>
          ]}}
        Random Seed=0
       )
       Alpha=$Particles.SampleFront
      )
     )
     2=&Lerp_Vector(
      A=&SelectVectorFromArray(
       Array Sampling Mode=::NewEnumerator0
       Direct Array Index=?
       Vector Selection Array=DataInterface() {propertyValues {FloatData [
          <-14.6878 1.5317 2.1723>
          <-11.2085 4.0251 1.37>
          <-12.3362 3.786 -1.2259>
          <-17.4229 3.299 -0.7355>
          <-18.5772 3.6156 0.1431>
          <-8.041 4.2384 -1.237>
          <-7.8193 4.7259 -0.2386>
          <-6.9166 4.1356 2.1358>
          <-2.6006 2.7113 -2.9742>
          <-2.6281 2.496 -2.527>
         ]}}
       Random Seed=0
      )
      B=&SelectVectorFromArray(
       Array Sampling Mode=::NewEnumerator0
       Direct Array Index=?
       Vector Selection Array=DataInterface() {propertyValues {FloatData [
          <-16.9533 -1.3057 -2.3836>
          <-7.4976 2.7844 -3.0039>
          <-14.7555 -2.3148 1.3283>
          <-16.6911 -1.9617 1.7466>
          <-19.8401 -2.2183 1.5147>
          <-22.4783 2.647 -1.626>
          <-11.0984 -1.3589 2.6617>
         ]}}
       Random Seed=0
      )
      Alpha=$Particles.SampleFront
     )
     3=&Lerp_Vector(
      A=&SelectVectorFromArray(
       Array Sampling Mode=::NewEnumerator0
       Direct Array Index=?
       Vector Selection Array=DataInterface() {propertyValues {FloatData [
          <-23.6388 -7.9812 2.4757>
          <-12.9163 -7.2094 4.9522>
          <-37.9449 -5.1975 2.6928>
          <-36.2232 -6.8574 -2.5344>
          <-13.2736 -7.0455 -5.6641>
          <-12.5909 -10.5002 -0.7843>
         ]}}
       Random Seed=0
      )
      B=&SelectVectorFromArray(
       Array Sampling Mode=::NewEnumerator0
       Direct Array Index=?
       Vector Selection Array=DataInterface() {propertyValues {FloatData [
          <-14.969 0.8682 3.0174>
          <-25.8421 2.5527 -0.5828>
          <-30.6595 0.7918 1.5502>
          <-19.7216 2.0127 -3.4952>
          <-17.7746 -1.4795 -5.9624>
          <-7.0841 -4.933 -6.6612>
         ]}}
       Random Seed=0
      )
      Alpha=$Particles.SampleFront
     )
     4=&Lerp_Vector(
      A=&SelectVectorFromArray(
       Array Sampling Mode=::NewEnumerator0
       Direct Array Index=?
       Vector Selection Array=DataInterface() {propertyValues {FloatData [
          <27.4419 6.5103 4.0501>
          <11.8267 9.6405 1.9979>
          <14.7902 9.6904 1.1286>
          <14.8166 8.469 3.2404>
          <15.7343 9.5166 -0.8001>
          <36.8364 6.9585 -0.4607>
          <28.8563 4.0114 -3.687>
          <37.965 5.6305 -2.5754>
         ]}}
       Random Seed=0
      )
      B=&SelectVectorFromArray(
       Array Sampling Mode=::NewEnumerator0
       Direct Array Index=?
       Vector Selection Array=DataInterface() {propertyValues {FloatData [
          <11.4519 -0.0051 4.571>
          <14.7016 -1.0159 -2.8186>
          <22.8174 0.6917 -3.7761>
          <24.9856 3.3131 -4.5906>
          <34.6597 1.1761 -2.8052>
          <38.2054 -1.0153 -1.108>
          <15.5614 0.5898 5.7356>
         ]}}
       Random Seed=0
      )
      Alpha=$Particles.SampleFront
     )
     5=&Lerp_Vector(
      A=&SelectVectorFromArray(
       Array Sampling Mode=::NewEnumerator0
       Direct Array Index=?
       Vector Selection Array=DataInterface() {propertyValues {FloatData [
          <11.3608 -7.285 3.9829>
          <5.8052 -6.7008 -11.9312>
          <2.5847 -14.2609 6.5765>
          <6.8426 -8.2559 -11.6528>
          <1.9318 -14.828 -2.6023>
          <9.6134 -9.9725 0.0819>
          <10.4272 -5.6616 -13.6735>
         ]}}
       Random Seed=0
      )
      B=&SelectVectorFromArray(
       Array Sampling Mode=::NewEnumerator0
       Direct Array Index=?
       Vector Selection Array=DataInterface() {propertyValues {FloatData [
          <8.9191 5.4662 -5.4412>
          <15.1709 3.5811 -3.6124>
          <11.1829 15.3455 3.1824>
          <11.616 8.7533 11.3886>
          <10.9359 8.0466 11.5299>
          <8.524 8.0123 -5.9023>
          <10.1192 1.6416 12.6719>
          <12.1323 4.3892 2.4079>
         ]}}
       Random Seed=0
      )
      Alpha=$Particles.SampleFront
     )
     Index=&SpawnGroup()
     Num=::NewEnumerator4
    )
    Skeletal Mesh=$System.Skeletal Mesh
   )
   &AddVelocity(
    Additional Quaternion Rotation=false
    Rotation Mode=::NewEnumerator3
    Rotation Quaternion:Quat4f=$Particles.SkeletalMeshLocation001.SampledRotation
    Velocity=&MakeVector(
     X=&RandomRangeFloat(Minimum=20, Maximum=25)
     Y=&RandomRangeFloat(Minimum=-18.5, Maximum=-32.5)
     Z=0
    )
    Velocity Speed Scale=&MergeHelper_FloatSwitch(
     0=&Lerp_Float(A=1, B=0, Alpha=&MakeFloatFromInt(Int=$Particles.PartitionParticles.Partition))
     1=0
     2=&Lerp_Float(A=1, B=0, Alpha=&MakeFloatFromInt(Int=$Particles.PartitionParticles.Partition))
     3=&Lerp_Float(A=1, B=0, Alpha=&MakeFloatFromInt(Int=$Particles.PartitionParticles.Partition))
     4=0
     Index=$Emitter.SpawnGroup
     Num=::NewEnumerator3
    )
   )
   &AddVelocity(
    Module.Use Falloff Distance=false
    Origin Offset=&RotateVectorByQuaternion(
     Vector To Rotate=&MergeHelper_VectorSwitch(
      0=<0 0 0>
      1=<16 0 0>
      2=<-16 0 0>
      3=<-27 -3 0>
      4=<27 3 0>
      5=<5 0 0>
      Index=&SpawnGroup()
      Num=::NewEnumerator4
     )
     Quaternion=$Particles.SkeletalMeshLocation.SampledRotation
    )
    Override Seed=false
    Velocity Mode=::NewEnumerator1
    Velocity Origin=$Output.SkeletalMeshLocation.SampledPositionWithoutOffset
    Velocity Speed=&MergeHelper_FloatSwitch(
     0=&Lerp_Float(
      A=0
      B=&RandomRangeFloat(Minimum=8, Maximum=15)
      Alpha=&MakeFloatFromInt(Int=$Particles.PartitionParticles.Partition)
     )
     1=0
     2=&Lerp_Float(
      A=0
      B=&RandomRangeFloat(Minimum=3, Maximum=6)
      Alpha=&MakeFloatFromInt(Int=$Particles.PartitionParticles.Partition)
     )
     3=&Lerp_Float(
      A=0
      B=&RandomRangeFloat(Minimum=3, Maximum=6)
      Alpha=&MakeFloatFromInt(Int=$Particles.PartitionParticles.Partition)
     )
     4=0
     Index=$Emitter.SpawnGroup
     Num=::NewEnumerator3
    )
   ) {module=AddVelocity001}
  ]
  particleUpdateScript [
   &ParticleState()
   &ScaleSpriteSize(Uniform Curve Sprite Scale=DataInterface() {propertyValues {Curve {keys [{arriveTangent=3.8099, interpMode=RCIM_Cubic, leaveTangent=3.8099, tangentMode=RCTM_User, time=0, value=0}, <at 0.4 1 RCIM_Cubic>]}}})
   &ScaleSpriteSize(
    Scale Sprite Size Mode=::NewEnumerator0
    Uniform Scale Factor=&FloatFromCurve(
     FloatCurve=DataInterface() {propertyValues {Curve {keys [
         {arriveTangent=1.0554, interpMode=RCIM_Cubic, leaveTangent=1.0554, tangentMode=RCTM_User, time=0, value=1}
         {arriveTangentWeight=0.3334, interpMode=RCIM_Cubic, tangentMode=RCTM_User, time=1, value=1.4}
        ]}}}
     CurveIndex=$System.AmplitudeIntensity
    )
   ) {module=ScaleSpriteSize001}
   &SpriteRotationRate(Rotation Rate=&Multiply_Float(
     A=$Particles.RotationDirection
     B=&Multiply_Float(
      A=&RandomRangeFloat(Minimum=20, Maximum=35)
      B=&MergeHelper_FloatSwitch(0=1, 1=0.8, 2=0, 3=0.8, Index=$Particles.VisibilityTag, Num=::NewEnumerator2)
     )
    ))
   &ScaleVelocity(
    Coordinate Space:ENiagaraCoordinateSpace=::World
    Velocity Scale=&VectorFromFloat(Value=&Lerp_Float(A=0.685, B=1.2, Alpha=$System.AmplitudeIntensity))
   )
   &ScaleColor(
    Scale Alpha=&FloatFromCurve(
     FloatCurve=DataInterface() {propertyValues {Curve {keys [{arriveTangent=-1.4967, interpMode=RCIM_Cubic, time=0.4, value=1}, {interpMode=RCIM_Cubic, leaveTangent=-3.0489, time=1, value=0}]}}}
     Scale Curve=$System.VelocityFade
    )
    Scale RGB=&Multiply_Vector(
     A=&MakeVectorFromLinearColorRGB(Color=&HueShiftLinearColor(
       Hue=&Lerp_Float(
        A=0
        B=/BRCosmetics/Weapons/FORT_Melee/Pickaxe_GateHound/FX/VariantManager/VariantFloatSelector(
         Original Float:NiagaraFloat=&UniformAOrBFloat(A=-0.0125, B=0.05, Evaluation Type=::NewEnumerator0, Recalculate Random Each Loop=true, Override Seed=false)
         Variant 1 Float:NiagaraFloat=0
         Variant 2 Float:NiagaraFloat=?
         Variant 3 Float:NiagaraFloat=?
         Variant Amount:NiagaraInt32=0
        )
        Alpha=$System.AmplitudeIntensity
       )
       Saturation=1
       Value=&FloatFromCurve(
        FloatCurve=DataInterface() {propertyValues {Curve {keys [{arriveTangent=1.9866, interpMode=RCIM_Cubic, leaveTangent=1.9866, tangentMode=RCTM_User, time=0, value=1}, <at 1 2 RCIM_Cubic>]}}}
        CurveIndex=$System.AmplitudeIntensity
       )
       Linear Color=&VariantColorSelector(Variant Amount=0, Original Color=<rgba 1 0.1274 0.402 1>, Variant 1 Color=<rgba 1 0.0953 0.08 1>, Variant 2 Color=?)
      ))
     B=&MakeVectorFromLinearColorRGB(Color=$System.TODColor)
    )
    ScaleA=true
    ScaleRGB=true
   )
   &FN_Sprite_Attributes(
    Color=&MakeLinearColorFromVectorAndFloat(
     "Vector (RGB)"=&MakeVectorFromLinearColorRGB(Color=$Particles.Color)
     "Float (Alpha)"=&FloatFromCurve(FloatCurve=DataInterface() {propertyValues {Curve {keys [<at 0.5 1>,<at 1 0>]}}}, Scale Curve=0.04)
    )
    Create Color Attribute=true
    Create SpriteSize Attribute=true
    SpriteSize=&Multiply_Vector2D_ByFloat(Float=2.25, Vector2D=$Particles.SpriteSize)
    Write to Module Namespace=true
   ) {module=FN_Sprite_Attributes003}
   &FN_Sprite_Attributes(
    Color=&MakeLinearColorFromVectorAndFloat(
     "Vector (RGB)"=&Multiply_VectorByFloat(
      Vector=&MakeVectorFromLinearColorRGB(Color=<rgba 3 3 3 1>)
      Float=&Multiply_Float(A=&Lerp_Float(A=1, B=3, Alpha=$System.AmplitudeIntensity), B=$System.TODScalar)
     )
     "Float (Alpha)"=&FloatFromCurve(
      FloatCurve=DataInterface() {propertyValues {Curve {keys [
          {arriveTangent=-8.3354, interpMode=RCIM_Cubic, leaveTangent=-8.3354, tangentMode=RCTM_User, time=0.1, value=3}
          {interpMode=RCIM_Cubic, leaveTangent=-2.9611, time=1, value=0}
         ]}}}
      Scale Curve=&Multiply_Float(
       A=$System.VelocityFade
       B=&Lerp_Float(
        A=1
        B=&UniformAOrBFloat(A=0, Evaluation Type=::NewEnumerator1, Recalculate Random Each Loop=true, Override Seed=false)
        Alpha=&FloatFromCurve(FloatCurve=DataInterface() {propertyValues {Curve {keys [<at 0.225 1>,<at 0.6 0>]}}})
       )
      )
     )
    )
    Create Color Attribute=true
    Create SpriteSize Attribute=true
    SpriteSize=&Multiply_Vector2D_ByFloat(
     Float=&Lerp_Float(A=1, B=1.2, Alpha=$System.AmplitudeIntensity)
     Vector2D=&Vector2DFromFloat(Value=&RandomRangeFloat(Minimum=3, Maximum=5.5))
    )
    Write to Module Namespace=true
   ) {module=FN_Sprite_Attributes001}
   &FN_Sprite_Attributes(
    Color=&MakeLinearColorFromVectorAndFloat(
     "Vector (RGB)"=&Multiply_VectorByFloat(Vector=&MakeVectorFromLinearColorRGB(Color=<rgba 3 3 3 1>), Float=$System.TODScalar)
     "Float (Alpha)"=&FloatFromCurve(
      FloatCurve=DataInterface() {propertyValues {Curve {keys [
          {arriveTangent=-8.3354, interpMode=RCIM_Cubic, leaveTangent=-8.3354, tangentMode=RCTM_User, time=0.1, value=3}
          {interpMode=RCIM_Cubic, leaveTangent=-2.9611, time=1, value=0}
         ]}}}
      Scale Curve=&Multiply_Float(
       A=$System.VelocityFade
       B=&Lerp_Float(
        A=1
        B=&UniformAOrBFloat(A=0, Evaluation Type=::NewEnumerator1, Recalculate Random Each Loop=true, Override Seed=false)
        Alpha=&FloatFromCurve(FloatCurve=DataInterface() {propertyValues {Curve {keys [<at 0.225 1>,<at 0.6 0>]}}})
       )
      )
     )
    )
    Create Color Attribute=true
    Create SpriteSize Attribute=true
    SpriteSize=&Multiply_Vector2D_ByFloat(
     Float=&Lerp_Float(A=1, B=1.2, Alpha=$System.AmplitudeIntensity)
     Vector2D=&Vector2DFromFloat(Value=&RandomRangeFloat(Minimum=3, Maximum=5.5))
    )
    Write to Module Namespace=true
   ) {module=FN_Sprite_Attributes002}
   /SetVariables(Particles.DynamicParametersAlt:Vector4f=&MakeVector4(
     X=&FloatFromCurve(FloatCurve=DataInterface() {propertyValues {Curve {keys [<at 0 2 RCIM_Cubic>, {arriveTangent=7.0068, interpMode=RCIM_Cubic, leaveTangent=7.0068, tangentMode=RCTM_User, time=1, value=6}]}}})
     Y=0
     Z=0
     W=0
    )) {isSetParameters=true}
   &FN_Sprite_Attributes(
    Color=&MakeLinearColorFromVectorAndFloat(
     "Vector (RGB)"=&Multiply_VectorByFloat(
      Vector=&MakeVectorFromLinearColorRGB(Color=<rgba 3 3 3 1>)
      Float=&Multiply_Float(A=&Lerp_Float(A=1, B=5, Alpha=$System.AmplitudeIntensity), B=$System.TODScalar)
     )
     "Float (Alpha)"=&FloatFromCurve(
      FloatCurve=DataInterface() {propertyValues {Curve {keys [{arriveTangent=-1.5731, interpMode=RCIM_Cubic, time=0.4, value=1}, {interpMode=RCIM_Cubic, leaveTangent=-2.9611, time=1, value=0}]}}}
      Scale Curve=&Lerp_Float(
       A=$System.VelocityFade
       B=&UniformAOrBFloat(A=0, Evaluation Type=::NewEnumerator1, Recalculate Random Each Loop=true, Override Seed=false)
       Alpha=&FloatFromCurve(FloatCurve=DataInterface() {propertyValues {Curve {keys [<at 0.4 1>,<at 1 0>]}}})
      )
     )
    )
    Create Color Attribute=true
    Create SpriteSize Attribute=true
    SpriteSize=&Multiply_Vector2D_ByFloat(
     Float=&Lerp_Float(A=1, B=1.2, Alpha=$System.AmplitudeIntensity)
     Vector2D=&Multiply_Vector2D(
      A=&Vector2DFromFloat(Value=&RandomRangeFloat(Minimum=5.5, Maximum=9))
      B=&Vector2DFromCurve(
       Vector2Curve=DataInterface() {propertyValues {
         XCurve {keys [{arriveTangent=-0.7214, interpMode=RCIM_Cubic, leaveTangent=-0.7214, tangentMode=RCTM_User, time=0.25, value=1}, <at 1 0.3>]}
         YCurve {keys [
           <at 0 1 RCIM_Cubic>
           {arriveTangent=-2.1615, interpMode=RCIM_Cubic, leaveTangent=-2.1615, tangentMode=RCTM_User, time=0.15, value=1.4}
           <at 1 0 RCIM_Cubic>
          ]}
        }}
       ScaleCurve=<1 1>
      )
     )
    )
    Write to Module Namespace=true
   )
   &SolveForcesAndVelocity(Clamp Velocity=false, Limit Acceleration=false)
   &KillParticles(Kill Particles=&Boolean_OrOperation(
     Bool1=&SetBoolByFloatComparison(A=$System.VelocityFade, B=0.01, Comparison Type=::NewEnumerator7)
     Bool2=&Boolean_OrOperation(
      Bool1=&Boolean_AndOperation(
       Bool1=$User.bisHandHidden
       Bool2=&Boolean_AndOperation(
        Bool1=/Niagara/DynamicInputs/Helpers/MaskBool(
         Mask Source:NiagaraInt32=1
         Mask Parameter:NiagaraInt32=$Particles.VisibilityTag
         Mask:NiagaraInt32=1
         Masked Bool:NiagaraBool=false
         Passthrough Bool:NiagaraBool=true
        )
        Bool2=&Boolean_OrOperation(
         Bool1=&SetBoolByIntComparison(A=&SpawnGroup(), B=1, Comparison Type=::NewEnumerator4)
         Bool2=&SetBoolByIntComparison(A=&SpawnGroup(), B=2, Comparison Type=::NewEnumerator4)
        )
       )
      )
      Bool2=&Boolean_OrOperation(
       Bool1=&Boolean_AndOperation(
        Bool1=$User.bisFootHidden
        Bool2=&Boolean_AndOperation(
         Bool1=&MaskBool(
          Mask Source:NiagaraInt32=1
          Mask Parameter:NiagaraInt32=$Particles.VisibilityTag
          Mask:NiagaraInt32=1
          Masked Bool:NiagaraBool=false
          Passthrough Bool:NiagaraBool=true
         )
         Bool2=&Boolean_OrOperation(
          Bool1=&SetBoolByIntComparison(A=&SpawnGroup(), B=3, Comparison Type=::NewEnumerator4)
          Bool2=&SetBoolByIntComparison(A=&SpawnGroup(), B=4, Comparison Type=::NewEnumerator4)
         )
        )
       )
       Bool2=&Boolean_AndOperation(
        Bool1=&Boolean_OrOperation(Bool1=$User.bisHeadHidden, Bool2=$System.bisHatHidden)
        Bool2=&Boolean_AndOperation(
         Bool1=&MaskBool(
          Mask Source:NiagaraInt32=1
          Mask Parameter:NiagaraInt32=$Particles.VisibilityTag
          Mask:NiagaraInt32=1
          Masked Bool:NiagaraBool=false
          Passthrough Bool:NiagaraBool=true
         )
         Bool2=&SetBoolByIntComparison(A=&SpawnGroup(), B=5, Comparison Type=::NewEnumerator4)
        )
       )
      )
     )
    ))
  ]
  rendererClasses [&NiagaraSpriteRendererProperties]
  renderers [
   {rendererClass=&NiagaraSpriteRendererProperties}
   {rendererClass=&NiagaraSpriteRendererProperties, rendererIndex=1}
   {rendererClass=&NiagaraSpriteRendererProperties, rendererIndex=2}
   {rendererClass=&NiagaraSpriteRendererProperties, rendererIndex=3}
   {rendererClass=&NiagaraSpriteRendererProperties, rendererIndex=4}
  ]
 }}
summary {
 systemName=NS_Character_ForestJeanWink_Idle
 emitters [{emitterName=Stars_Packed, rendererClasses [&NiagaraSpriteRendererProperties]}]
 userVariables [
  User.bisFootHidden {type=NiagaraBool}
  User.bIsFrontEnd {type=NiagaraBool}
  User.bIsFrontEndPreview {type=NiagaraBool}
  User.bisHandHidden {type=NiagaraBool}
  User.bisHeadHidden {type=NiagaraBool}
  User.DebugReactivity {type=NiagaraBool}
  User.HatEnabled {type=NiagaraFloat, defaultValue=1, description=""}
  User.Variant {type=NiagaraFloat, defaultValue=0, description=""}
 ]
}
