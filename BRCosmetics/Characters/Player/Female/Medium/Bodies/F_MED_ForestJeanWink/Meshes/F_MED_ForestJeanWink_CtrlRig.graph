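# Control Rig for the ForestJeanWink female medium body cosmetic, posing its secondary deform and tassel base bones.
# It runs the shared DeformRig function, then a Constraints function that weights torso and leg tassel bases across nearby spine, clavicle, pec, pelvis, groin, glute and thigh bones.
# Thigh Y and Z rotations, measured against pelvis-parented nulls, blend in corrective leg tassel rotations.
# Spherical pose readers on the clavicles push out the lat deform bones.
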ControlRigBlueprint /BRCosmetics/Characters/Player/Female/Medium/Bodies/F_MED_ForestJeanWink/Meshes/F_MED_ForestJeanWink_CtrlRig
graphs [
 New Function_ContainedGraph {nodes [
   n1 = RigVMFunctionEntryNode()
   RigVMFunction_Sequence = &RigVMFunction_Sequence(ExecuteContext=@n1.ExecuteContext)
   &RigUnit_ParentConstraint(
    AdvancedSettings=(RotationOrderForFilter=XZY)
    Child=(Type=Bone,Name="base_torsoTassels_02_l")
    Parents=((Item=(Type=Bone,Name="spine_01"),Weight=0.0142),(Item=(Type=Bone,Name="deform_pec_l"),Weight=0.0611),(Item=(Type=Bone,Name="spine_03"),Weight=0.0717),(Item=(Type=Bone,Name="spine_02"),Weight=0.0847),(Item=(Type=Bone,Name="spine_04"),Weight=0.7683))
    Weight=1
    ExecutePin=@RigVMFunction_Sequence.A
   )
   > &RigUnit_ParentConstraint(
    AdvancedSettings=(RotationOrderForFilter=XZY)
    Child=(Type=Bone,Name="base_torsoTassels_03_l")
    Parents="((Item=(Type=Bone,Name=\"spine_01\"),Weight=0.0007),(Item=(Type=Bone,Name=\"clavicle_l\"),Weight=0.0296),(Item=(Type=Bone,Name=\"spine_05\"),Weight=0.0476),(Item=(Type=Bone,Name=\"spine_03\"),Weight=0.0551),(Item=(Type=Bone,Name=\"spine_02\"),Weight=0.0671),(Item=(T ...(truncated)"
    Weight=1
   )
   > &RigUnit_ParentConstraint(
    AdvancedSettings=(RotationOrderForFilter=XZY)
    Child=(Type=Bone,Name="base_torsoTassels_01_l")
    Parents="((Item=(Type=Bone,Name=\"spine_01\"),Weight=0.0016),(Item=(Type=Bone,Name=\"spine_05\"),Weight=0.0029),(Item=(Type=Bone,Name=\"deform_pec_l\"),Weight=0.0267),(Item=(Type=Bone,Name=\"spine_03\"),Weight=0.0671),(Item=(Type=Bone,Name=\"spine_02\"),Weight=0.0713),(Item= ...(truncated)"
    Weight=1
   )
   > &RigUnit_ParentConstraint(
    AdvancedSettings=(RotationOrderForFilter=XZY)
    Child=(Type=Bone,Name="base_torsoTassels_01_m")
    Parents=((Item=(Type=Bone,Name="deform_pec_r"),Weight=0.0028),(Item=(Type=Bone,Name="spine_02"),Weight=0.0227),(Item=(Type=Bone,Name="spine_03"),Weight=0.0249),(Item=(Type=Bone,Name="spine_05"),Weight=0.0625),(Item=(Type=Bone,Name="spine_04"),Weight=0.8872))
    Weight=1
   )
   > &RigUnit_ParentConstraint(
    AdvancedSettings=(RotationOrderForFilter=XZY)
    Child=(Type=Bone,Name="base_torsoTassels_01_r")
    Parents="((Item=(Type=Bone,Name=\"spine_01\"),Weight=0.0018),(Item=(Type=Bone,Name=\"spine_05\"),Weight=0.0026),(Item=(Type=Bone,Name=\"deform_pec_r\"),Weight=0.0269),(Item=(Type=Bone,Name=\"spine_03\"),Weight=0.0674),(Item=(Type=Bone,Name=\"spine_02\"),Weight=0.0717),(Item= ...(truncated)"
    Weight=1
   )
   > &RigUnit_ParentConstraint(
    AdvancedSettings=(RotationOrderForFilter=XZY)
    Child=(Type=Bone,Name="base_torsoTassels_02_r")
    Parents=((Item=(Type=Bone,Name="spine_01"),Weight=0.0141),(Item=(Type=Bone,Name="deform_pec_r"),Weight=0.0598),(Item=(Type=Bone,Name="spine_03"),Weight=0.0718),(Item=(Type=Bone,Name="spine_02"),Weight=0.0846),(Item=(Type=Bone,Name="spine_04"),Weight=0.7698))
    Weight=1
   )
   > &RigUnit_ParentConstraint(
    AdvancedSettings=(RotationOrderForFilter=XZY)
    Child=(Type=Bone,Name="base_torsoTassels_03_r")
    Parents="((Item=(Type=Bone,Name=\"spine_01\"),Weight=0.0007),(Item=(Type=Bone,Name=\"clavicle_r\"),Weight=0.0296),(Item=(Type=Bone,Name=\"spine_05\"),Weight=0.0476),(Item=(Type=Bone,Name=\"spine_03\"),Weight=0.0551),(Item=(Type=Bone,Name=\"spine_02\"),Weight=0.0671),(Item=(T ...(truncated)"
    Weight=1
   )
   > &RigUnit_ParentConstraint(
    AdvancedSettings=(RotationOrderForFilter=XZY)
    Child=(Type=Bone,Name="base_tasselLeg_01_l")
    Parents="((Item=(Type=Bone,Name=\"deform_groin_r\"),Weight=0.0008),(Item=(Type=Bone,Name=\"thigh_twist_01_r\"),Weight=0.0204),(Item=(Type=Bone,Name=\"thigh_l\"),Weight=0.0229),(Item=(Type=Bone,Name=\"deform_groin_l\"),Weight=0.2536),(Item=(Type=Bone,Name=\"thigh_twist_01_l\" ...(truncated)"
    Weight=1
   )
   > &RigUnit_ParentConstraint(
    AdvancedSettings=(RotationOrderForFilter=XZY)
    Child=(Type=Bone,Name="base_tasselLeg_02_l")
    Parents="((Item=(Type=Bone,Name=\"thigh_twist_01_r\"),Weight=0.0047),(Item=(Type=Bone,Name=\"thigh_l\"),Weight=0.0182),(Item=(Type=Bone,Name=\"deform_groin_l\"),Weight=0.2933),(Item=(Type=Bone,Name=\"thigh_twist_01_l\"),Weight=0.3239),(Item=(Type=Bone,Name=\"pelvis\"),Weight ...(truncated)"
    Weight=1
   )
   > &RigUnit_ParentConstraint(
    AdvancedSettings=(RotationOrderForFilter=XZY)
    Child=(Type=Bone,Name="base_tasselLeg_03_l")
    Parents=((Item=(Type=Bone,Name="thigh_l"),Weight=0.0091),(Item=(Type=Bone,Name="deform_groin_l"),Weight=0.2946),(Item=(Type=Bone,Name="thigh_twist_01_l"),Weight=0.3103),(Item=(Type=Bone,Name="pelvis"),Weight=0.386))
    Weight=1
   )
   > &RigUnit_ParentConstraint(
    AdvancedSettings=(RotationOrderForFilter=XZY)
    Child=(Type=Bone,Name="base_tasselLeg_04_l")
    Parents="((Item=(Type=Bone,Name=\"deform_glute_l\"),Weight=0.0016),(Item=(Type=Bone,Name=\"thigh_l\"),Weight=0.0079),(Item=(Type=Bone,Name=\"deform_groin_l\"),Weight=0.2167),(Item=(Type=Bone,Name=\"thigh_twist_01_l\"),Weight=0.3629),(Item=(Type=Bone,Name=\"pelvis\"),Weight=0 ...(truncated)"
    Weight=1
   )
   > &RigUnit_ParentConstraint(
    AdvancedSettings=(RotationOrderForFilter=XZY)
    Child=(Type=Bone,Name="base_tasselLeg_05_l")
    Parents="((Item=(Type=Bone,Name=\"thigh_l\"),Weight=0.0167),(Item=(Type=Bone,Name=\"deform_glute_l\"),Weight=0.018),(Item=(Type=Bone,Name=\"deform_groin_l\"),Weight=0.0934),(Item=(Type=Bone,Name=\"pelvis\"),Weight=0.4344),(Item=(Type=Bone,Name=\"thigh_twist_01_l\"),Weight=0. ...(truncated)"
    Weight=1
   )
   > &RigUnit_ParentConstraint(
    AdvancedSettings=(RotationOrderForFilter=XZY)
    Child=(Type=Bone,Name="base_tasselLeg_06_l")
    Parents="((Item=(Type=Bone,Name=\"thigh_l\"),Weight=0.0169),(Item=(Type=Bone,Name=\"deform_groin_l\"),Weight=0.019),(Item=(Type=Bone,Name=\"deform_glute_l\"),Weight=0.0764),(Item=(Type=Bone,Name=\"pelvis\"),Weight=0.4126),(Item=(Type=Bone,Name=\"thigh_twist_01_l\"),Weight=0. ...(truncated)"
    Weight=1
   )
   > &RigUnit_ParentConstraint(
    AdvancedSettings=(RotationOrderForFilter=XZY)
    Child=(Type=Bone,Name="base_tasselLeg_07_l")
    Parents=((Item=(Type=Bone,Name="thigh_l"),Weight=0.0088),(Item=(Type=Bone,Name="deform_glute_l"),Weight=0.1937),(Item=(Type=Bone,Name="pelvis"),Weight=0.3554),(Item=(Type=Bone,Name="thigh_twist_01_l"),Weight=0.4421))
    Weight=1
   )
   > &RigUnit_ParentConstraint(
    AdvancedSettings=(RotationOrderForFilter=XZY)
    Child=(Type=Bone,Name="base_tasselLeg_08_l")
    Parents=((Item=(Type=Bone,Name="thigh_l"),Weight=0.0017),(Item=(Type=Bone,Name="pelvis"),Weight=0.2903),(Item=(Type=Bone,Name="deform_glute_l"),Weight=0.3324),(Item=(Type=Bone,Name="thigh_twist_01_l"),Weight=0.3756))
    Weight=1
   )
   > &RigUnit_ParentConstraint(
    AdvancedSettings=(RotationOrderForFilter=XZY)
    Child=(Type=Bone,Name="base_tasselLeg_09_l")
    Parents=((Item=(Type=Bone,Name="thigh_l"),Weight=0.0034),(Item=(Type=Bone,Name="pelvis"),Weight=0.254),(Item=(Type=Bone,Name="thigh_twist_01_l"),Weight=0.3406),(Item=(Type=Bone,Name="deform_glute_l"),Weight=0.4021))
    Weight=1
   )
   > &RigUnit_ParentConstraint(
    AdvancedSettings=(RotationOrderForFilter=XZY)
    Child=(Type=Bone,Name="base_tasselLeg_10_l")
    Parents="((Item=(Type=Bone,Name=\"deform_glute_r\"),Weight=0.0011),(Item=(Type=Bone,Name=\"thigh_twist_01_r\"),Weight=0.008),(Item=(Type=Bone,Name=\"thigh_l\"),Weight=0.0444),(Item=(Type=Bone,Name=\"deform_glute_l\"),Weight=0.2549),(Item=(Type=Bone,Name=\"pelvis\"),Weight=0. ...(truncated)"
    Weight=1
   )
   > &RigUnit_ParentConstraint(
    AdvancedSettings=(RotationOrderForFilter=XZY)
    Child=(Type=Bone,Name="base_tasselLeg_01_r")
    Parents="((Item=(Type=Bone,Name=\"deform_groin_l\"),Weight=0.0008),(Item=(Type=Bone,Name=\"thigh_twist_01_l\"),Weight=0.0204),(Item=(Type=Bone,Name=\"thigh_r\"),Weight=0.0229),(Item=(Type=Bone,Name=\"deform_groin_r\"),Weight=0.2536),(Item=(Type=Bone,Name=\"thigh_twist_01_r\" ...(truncated)"
    Weight=1
   )
   > &RigUnit_ParentConstraint(
    AdvancedSettings=(RotationOrderForFilter=XZY)
    Child=(Type=Bone,Name="base_tasselLeg_02_r")
    Parents="((Item=(Type=Bone,Name=\"thigh_twist_01_l\"),Weight=0.0047),(Item=(Type=Bone,Name=\"thigh_r\"),Weight=0.0182),(Item=(Type=Bone,Name=\"deform_groin_r\"),Weight=0.2933),(Item=(Type=Bone,Name=\"thigh_twist_01_r\"),Weight=0.3239),(Item=(Type=Bone,Name=\"pelvis\"),Weight ...(truncated)"
    Weight=1
   )
   > &RigUnit_ParentConstraint(
    AdvancedSettings=(RotationOrderForFilter=XZY)
    Child=(Type=Bone,Name="base_tasselLeg_03_r")
    Parents=((Item=(Type=Bone,Name="thigh_r"),Weight=0.0091),(Item=(Type=Bone,Name="deform_groin_r"),Weight=0.2946),(Item=(Type=Bone,Name="thigh_twist_01_r"),Weight=0.3103),(Item=(Type=Bone,Name="pelvis"),Weight=0.386))
    Weight=1
   )
   > &RigUnit_ParentConstraint(
    AdvancedSettings=(RotationOrderForFilter=XZY)
    Child=(Type=Bone,Name="base_tasselLeg_04_r")
    Parents="((Item=(Type=Bone,Name=\"deform_glute_r\"),Weight=0.0016),(Item=(Type=Bone,Name=\"thigh_r\"),Weight=0.0079),(Item=(Type=Bone,Name=\"deform_groin_r\"),Weight=0.2167),(Item=(Type=Bone,Name=\"thigh_twist_01_r\"),Weight=0.3629),(Item=(Type=Bone,Name=\"pelvis\"),Weight=0 ...(truncated)"
    Weight=1
   )
   > &RigUnit_ParentConstraint(
    AdvancedSettings=(RotationOrderForFilter=XZY)
    Child=(Type=Bone,Name="base_tasselLeg_05_r")
    Parents="((Item=(Type=Bone,Name=\"thigh_r\"),Weight=0.0167),(Item=(Type=Bone,Name=\"deform_glute_r\"),Weight=0.018),(Item=(Type=Bone,Name=\"deform_groin_r\"),Weight=0.0934),(Item=(Type=Bone,Name=\"pelvis\"),Weight=0.4344),(Item=(Type=Bone,Name=\"thigh_twist_01_r\"),Weight=0. ...(truncated)"
    Weight=1
   )
   > &RigUnit_ParentConstraint(
    AdvancedSettings=(RotationOrderForFilter=XZY)
    Child=(Type=Bone,Name="base_tasselLeg_06_r")
    Parents="((Item=(Type=Bone,Name=\"thigh_r\"),Weight=0.0169),(Item=(Type=Bone,Name=\"deform_groin_r\"),Weight=0.019),(Item=(Type=Bone,Name=\"deform_glute_r\"),Weight=0.0764),(Item=(Type=Bone,Name=\"pelvis\"),Weight=0.4126),(Item=(Type=Bone,Name=\"thigh_twist_01_r\"),Weight=0. ...(truncated)"
    Weight=1
   )
   > &RigUnit_ParentConstraint(
    AdvancedSettings=(RotationOrderForFilter=XZY)
    Child=(Type=Bone,Name="base_tasselLeg_07_r")
    Parents=((Item=(Type=Bone,Name="thigh_r"),Weight=0.0088),(Item=(Type=Bone,Name="deform_glute_r"),Weight=0.1937),(Item=(Type=Bone,Name="pelvis"),Weight=0.3554),(Item=(Type=Bone,Name="thigh_twist_01_r"),Weight=0.4421))
    Weight=1
   )
   > &RigUnit_ParentConstraint(
    AdvancedSettings=(RotationOrderForFilter=XZY)
    Child=(Type=Bone,Name="base_tasselLeg_08_r")
    Parents=((Item=(Type=Bone,Name="thigh_r"),Weight=0.0017),(Item=(Type=Bone,Name="pelvis"),Weight=0.2903),(Item=(Type=Bone,Name="deform_glute_r"),Weight=0.3324),(Item=(Type=Bone,Name="thigh_twist_01_r"),Weight=0.3756))
    Weight=1
   )
   > &RigUnit_ParentConstraint(
    AdvancedSettings=(RotationOrderForFilter=XZY)
    Child=(Type=Bone,Name="base_tasselLeg_09_r")
    Parents=((Item=(Type=Bone,Name="thigh_r"),Weight=0.0034),(Item=(Type=Bone,Name="pelvis"),Weight=0.254),(Item=(Type=Bone,Name="thigh_twist_01_r"),Weight=0.3406),(Item=(Type=Bone,Name="deform_glute_r"),Weight=0.4021))
    Weight=1
   )
   > &RigUnit_ParentConstraint(
    AdvancedSettings=(RotationOrderForFilter=XZY)
    Child=(Type=Bone,Name="base_tasselLeg_10_r")
    Parents="((Item=(Type=Bone,Name=\"deform_glute_l\"),Weight=0.0011),(Item=(Type=Bone,Name=\"thigh_twist_01_l\"),Weight=0.008),(Item=(Type=Bone,Name=\"thigh_r\"),Weight=0.0444),(Item=(Type=Bone,Name=\"deform_glute_r\"),Weight=0.2549),(Item=(Type=Bone,Name=\"pelvis\"),Weight=0. ...(truncated)"
    Weight=1
   )
   RigVMFunctionReturnNode(ExecuteContext=@RigVMFunction_Sequence.B)
  ]}
 New Function_ContainedGraph {nodes [
   n1 = RigVMFunctionEntryNode()
   RigVMFunction_MathQuaternionToEuler = &RigVMFunction_MathQuaternionToEuler(
    RotationOrder=XYZ
    Value=&RigUnit_GetRelativeTransformForItem(
     Child=(Type=Bone,Name="hand_l")
     Parent=(Type=Null,Name="lowerarm_l")
     bChildInitial=False
     bParentInitial=False
     Child.Name=@n1.Child
     Parent.Name=@n1.Parent
    ).RelativeTransform.Rotation
   )
   Remap = &RigVMFunction_MathDoubleRemap(
    TargetMaximum=1
    TargetMinimum=0
    Value=@RigVMFunction_MathQuaternionToEuler.Result.Z
    SourceMaximum=@n1.Source Maximum
    SourceMinimum=@n1.Source Minimum
   )
   > RigVMFunctionReturnNode(Delta Rotation=@Remap.Result)
  ]}
 New Function_ContainedGraph {nodes [
   n1 = RigVMFunctionEntryNode()
   RigVMFunction_MathQuaternionToEuler = &RigVMFunction_MathQuaternionToEuler(
    RotationOrder=XYZ
    Value=&RigUnit_GetRelativeTransformForItem(
     Child=(Type=Bone,Name="hand_l")
     Parent=(Type=Null,Name="lowerarm_l")
     bChildInitial=False
     bParentInitial=False
     Child.Name=@n1.Child
     Parent.Name=@n1.Parent
    ).RelativeTransform.Rotation
   )
   Remap = &RigVMFunction_MathDoubleRemap(
    TargetMaximum=1
    TargetMinimum=0
    Value=@RigVMFunction_MathQuaternionToEuler.Result.Y
    SourceMaximum=@n1.Source Maximum
    SourceMinimum=@n1.Source Minimum
   )
   > RigVMFunctionReturnNode(Delta Rotation=@Remap.Result)
  ]}
 RigVMFunctionLibrary {nodes [
   Constraints = RigVMCollapseNode(ExecuteContext=())
   Delta Rotation Y = RigVMCollapseNode(ExecuteContext=())
   Delta Rotation Z = RigVMCollapseNode(ExecuteContext=())
  ]}
 RigVMModel {nodes [
   RigUnit_BeginExecution = &RigUnit_BeginExecution()
   FN_DeformRig_FNC = RigVMFunctionReferenceNode(
    DeformRig_ Curve="(deformRig_l_upperarm_up_alpha_72_#1=False,deformRig_r_upperarm_up_alpha_73_#2=False,deformRig_l_upperarm_fr_alpha_74_#3=False,deformRig_r_upperarm_fr_alpha_75_CB51B8 ...(truncated)"
    DeformRig_Bool="(upperarm_1_#4=True,elbow_3_#5=True,wrist_7_#6=True,groin_8_#7=True,glute_17_#8=True,knee_9_40C408324CA261 ...(truncated)"
    DeformRig_Debug=(upperarm_debug_18_#4=False,wrist_debug_20_#6=False,lat_debug_26_#9=false,groinGlute_debug_27_#7=False)
    DeformRig_Value="(upperarm_up_18_#10=(Rotation=(X=0.0,Y=-0.0,Z=0.0,W=1.0),Translation=(X=0.0,Y=0.0,Z=6),Scale3D=(X=0.7,Y=1.0,Z=1.0)),upperarm_fwd_fr_19_#11=(Rotation=(X=0,Y=0,Z=0,W=1),Translation=(X=-3.5,Y=10.0,Z=0. ...(truncated)"
    ExecuteContext=@RigUnit_BeginExecution.ExecutePin
   )
   > Constraints = RigVMFunctionReferenceNode()
   > Delta Rotation Z = RigVMFunctionReferenceNode(Child=thigh_r, Parent=thigh_r_null, Source Maximum=-100)
   > &RigUnit_ModifyTransforms(
    ItemToModify="((Item=(Type=Bone,Name=\"base_tasselLeg_01_r\"),Transform=(Rotation=(X=0,Y=0.4103,Z=0,W=-0.912),Translation=(X=0,Y=0,Z=0),Scale3D=(X=1,Y=1,Z=1))),(Item=(Type=Bone,Name=\"base_tasselLeg_02_r\"),Transform=(Rotation=(X=0.0743,Y=-0.381,Z=0.0523,W=0.9201),Translati ...(truncated)"
    WeightMinimum=0
    Weight=@"Delta Rotation Z".Delta Rotation
   )
   > Delta Rotation Z_1 = RigVMFunctionReferenceNode(Child=thigh_l, Parent=thigh_l_null, Source Maximum=-100)
   > &RigUnit_ModifyTransforms(
    ItemToModify="((Item=(Type=Bone,Name=\"base_tasselLeg_01_l\"),Transform=(Rotation=(X=0,Y=0.4103,Z=0,W=-0.912),Translation=(X=0,Y=0,Z=0),Scale3D=(X=1,Y=1,Z=1))),(Item=(Type=Bone,Name=\"base_tasselLeg_02_l\"),Transform=(Rotation=(X=-0.0743,Y=-0.381,Z=-0.0523,W=0.9201),Transla ...(truncated)"
    WeightMinimum=0
    Weight=@"Delta Rotation Z_1".Delta Rotation
   )
   > Delta Rotation Y = RigVMFunctionReferenceNode(Child=thigh_r, Parent=thigh_r_null, Source Maximum=-50)
   > &RigUnit_ModifyTransforms(
    ItemToModify="((Item=(Type=Bone,Name=\"base_tasselLeg_03_r\"),Transform=(Rotation=(X=0.0228,Y=-0.0722,Z=0.0016,W=0.9971),Translation=(X=0,Y=0,Z=0),Scale3D=(X=1,Y=1,Z=1))),(Item=(Type=Bone,Name=\"base_tasselLeg_04_r\"),Transform=(Rotation=(X=-0.0757,Y=0.248,Z=-0.0194,W=-0.96 ...(truncated)"
    WeightMinimum=0
    Weight=@"Delta Rotation Y".Delta Rotation
   )
   > Delta Rotation Y_1 = RigVMFunctionReferenceNode(Child=thigh_l, Parent=thigh_l_null, Source Maximum=-50)
   > &RigUnit_ModifyTransforms(
    ItemToModify="((Item=(Type=Bone,Name=\"base_tasselLeg_03_l\"),Transform=(Rotation=(X=-0.0228,Y=-0.0722,Z=-0.0016,W=0.9971),Translation=(X=0,Y=0,Z=0),Scale3D=(X=1,Y=1,Z=1))),(Item=(Type=Bone,Name=\"base_tasselLeg_04_l\"),Transform=(Rotation=(X=-0.0757,Y=-0.248,Z=-0.0194,W=0. ...(truncated)"
    WeightMinimum=0
    Weight=@"Delta Rotation Y_1".Delta Rotation
   )
   > n1 = &RigUnit_SphericalPoseReader(
    ActiveRegionScaleFactors=(PositiveWidth=0.3,NegativeWidth=1,PositiveHeight=1.5,NegativeHeight=1.5)
    ActiveRegionSize=0.1
    Debug=(bDrawDebug=false,bDraw2D=False,bDrawLocalAxes=False,DebugScale=25,DebugSegments=20,DebugThickness=0.25)
    DriverAxis=(X=1,Y=0,Z=0)
    DriverItem=(Type=Bone,Name="clavicle_l")
    FalloffRegionScaleFactors=(PositiveWidth=0.8,NegativeWidth=1,PositiveHeight=1,NegativeHeight=1)
    FalloffSize=0.3
    FlipHeightScaling=False
    FlipWidthScaling=False
    OptionalParentItem=(Type=Bone,Name="None")
    RotationOffset=(X=0,Y=-40,Z=0)
   )
   > &RigUnit_ModifyTransforms(
    ItemToModify=((Item=(Type=Bone,Name="deform_lat_l"),Transform=(Rotation=(X=0,Y=0,Z=0,W=1),Translation=(X=0,Y=0,Z=10),Scale3D=(X=1,Y=1,Z=1))))
    WeightMinimum=0
    Weight=@n1.OutputParam
   )
   > n2 = &RigUnit_SphericalPoseReader(
    ActiveRegionScaleFactors=(PositiveWidth=0.3,NegativeWidth=1,PositiveHeight=1.5,NegativeHeight=1.5)
    ActiveRegionSize=0.1
    Debug=(bDrawDebug=false,bDraw2D=False,bDrawLocalAxes=False,DebugScale=25,DebugSegments=20,DebugThickness=0.25)
    DriverAxis=(X=-1,Y=0,Z=0)
    DriverItem=(Type=Bone,Name="clavicle_r")
    FalloffRegionScaleFactors=(PositiveWidth=0.8,NegativeWidth=1,PositiveHeight=1,NegativeHeight=1)
    FalloffSize=0.3
    FlipHeightScaling=False
    FlipWidthScaling=False
    OptionalParentItem=(Type=Bone,Name="None")
    RotationOffset=(X=0,Y=140,Z=0)
   )
   > &RigUnit_ModifyTransforms(
    ItemToModify=((Item=(Type=Bone,Name="deform_lat_r"),Transform=(Rotation=(X=0,Y=0,Z=0,W=1),Translation=(X=0,Y=0,Z=-10),Scale3D=(X=1,Y=1,Z=1))))
    WeightMinimum=0
    Weight=@n2.OutputParam
   )
   &RigUnit_PrepareForExecution()
  ]}
]
hierarchy [
 root {loc=<0 0 0>, rot=<0 0 0 1>}
 pelvis {parent=root, loc=<0 -0.5301 98.0571>, rot=<0.0127 0.707 -0.0127 -0.707>}
 spine_01 {parent=pelvis, loc=<6.1208 -0.9328 0>, rot=<0 0 0.034 -0.9994>}
 spine_02 {parent=spine_01, loc=<6.5997 -0.2095 0>, rot=<0 0 0.0238 -0.9997>}
 spine_03 {parent=spine_02, loc=<6.4914 -0.1237 0>, rot=<0 0 0.0149 -0.9999>}
 spine_04 {parent=spine_03, loc=<6.9598 0.0794 0>, rot=<0 0 -0.0132 -0.9999>}
 spine_05 {parent=spine_04, loc=<6.9987 0 0>, rot=<0 0 -0.0015 -1>}
 neck_01 {parent=spine_05, loc=<11.3138 0.1529 0>, rot=<0 0 -0.2003 -0.9797>}
 neck_02 {parent=neck_01, loc=<4.5627 0 0>, rot=<0 0 0.0504 -0.9987>}
 head {parent=neck_02, loc=<5.3914 0 0>, rot=<0 0 0.115 -0.9934>}
 clavicle_l {parent=spine_05, loc=<5.4831 -0.1053 -3.2946>, rot=<0.0465 -0.7289 -0.009 -0.6829>}
 upperarm_l {parent=clavicle_l, loc=<9.8658 -1.2136 -0.4105>, rot=<0.0164 -0.3245 -0.0269 -0.9454>}
 lowerarm_l {parent=upperarm_l, loc=<26.465 -0.8345 0.259>, rot=<0.0036 0.0161 -0.2298 -0.9731>}
 hand_l {parent=lowerarm_l, loc=<22.7355 -0.3203 -0.6105>, rot=<-0.6784 -0.0223 0.0735 -0.7307>}
 thumb_01_l {parent=hand_l, loc=<1.3118 -0.3191 -2.459>, rot=<0.6546 -0.4042 0.0664 -0.6354>}
 thumb_02_l {parent=thumb_01_l, loc=<5.9336 -0.0448 0.0401>, rot=<-0.0037 0.0213 0.298 -0.9543>}
 thumb_03_l {parent=thumb_02_l, loc=<2.2223 0.3682 -0.0907>, rot=<0.0118 -0.013 0.0066 -0.9998>}
 index_metacarpal_l {parent=hand_l, loc=<1.6535 0.0183 -2.96>, rot=<0.0158 -0.0124 -0.0173 -0.9996>}
 index_01_l {parent=index_metacarpal_l, loc=<8.3008 0.0357 -0.0524>, rot=<-0.0044 0.0485 0.0924 -0.9945>}
 index_02_l {parent=index_01_l, loc=<3.3587 -0.3004 0.0452>, rot=<-0.002 0.0027 0.3225 -0.9466>}
 index_03_l {parent=index_02_l, loc=<2.5578 -0.123 -0.0252>, rot=<-0.002 0.0067 0.2558 -0.9667>}
 middle_metacarpal_l {parent=hand_l, loc=<2.1856 -0.1725 -0.5984>, rot=<0.006 0.0168 -0.0329 -0.9993>}
 middle_01_l {parent=middle_metacarpal_l, loc=<7.3275 -0.2447 0.0129>, rot=<-0.0208 0.0042 0.0858 -0.9961>}
 middle_02_l {parent=middle_01_l, loc=<4.0123 0.1401 0.1175>, rot=<-0.0218 0.0258 0.3512 -0.9357>}
 middle_03_l {parent=middle_02_l, loc=<2.6077 0.0534 0.1373>, rot=<0.0082 0.0188 0.1708 -0.9851>}
 ring_metacarpal_l {parent=hand_l, loc=<2.2074 -0.3625 1.3638>, rot=<-0.0284 0.0722 -0.0458 -0.9959>}
 ring_01_l {parent=ring_metacarpal_l, loc=<7.0295 -0.6197 -0.1446>, rot=<-0.0489 -0.0252 0.1892 -0.9804>}
 ring_02_l {parent=ring_01_l, loc=<3.6174 0.4647 -0.1027>, rot=<0 0 0.2806 -0.9598>}
 ring_03_l {parent=ring_02_l, loc=<2.4152 0.114 0.0199>, rot=<0 0 0.1189 -0.9929>}
 pinky_metacarpal_l {parent=hand_l, loc=<1.7616 -0.9753 3.3814>, rot=<-0.0397 0.0814 -0.069 -0.9935>}
 pinky_01_l {parent=pinky_metacarpal_l, loc=<7.1709 -0.4831 -0.1094>, rot=<-0.0654 -0.0035 0.2242 -0.9723>}
 pinky_02_l {parent=pinky_01_l, loc=<2.8274 0.0893 -0.1043>, rot=<0.0037 0.0011 0.2854 -0.9584>}
 pinky_03_l {parent=pinky_02_l, loc=<2.2046 0.0307 -0.0078>, rot=<-0.0002 0.0098 0.1445 -0.9895>}
 deform_wrist_dn_base_l {parent=hand_l, loc=<0 -1.5326 0>, rot=<0 0 0 -1>}
 deform_wrist_dn_l {parent=deform_wrist_dn_base_l, loc=<0 0 0>, rot=<0 0 0 1>}
 deform_wrist_up_base_l {parent=hand_l, loc=<0 2.3009 0>, rot=<0 0 0 -1>}
 deform_wrist_up_l {parent=deform_wrist_up_base_l, loc=<0 0 0>, rot=<0 0 0 1>}
 base_handTassels_01_l {parent=hand_l, loc=<2.3696 1.3123 -2.4744>, rot=<0.7627 -0.0266 -0.1807 -0.6204>}
 dyn_handTassels_01_l {parent=base_handTassels_01_l, loc=<0 0 0>, rot=<0 0 0 1>}
 base_handTassels_02_l {parent=hand_l, loc=<4.8796 2.3764 0.0941>, rot=<0.7037 0.0358 -0.02 -0.7093>}
 dyn_handTassels_02_l {parent=base_handTassels_02_l, loc=<0 0 0>, rot=<0 0 0 1>}
 base_handTassels_03_l {parent=hand_l, loc=<2.0738 1.5491 2.7522>, rot=<0.642 0.191 0.0062 -0.7425>}
 dyn_handTassels_03_l {parent=base_handTassels_03_l, loc=<0 0 0>, rot=<0 0 0 1>}
 lowerarm_twist_02_l {parent=lowerarm_l, loc=<10.6707 0 0>, rot=<0 0 0 1>}
 lowerarm_twist_01_l {parent=lowerarm_l, loc=<18.316 0 0>, rot=<0 0 0 1>}
 deform_elbow_in_base_l {parent=lowerarm_l, loc=<0.1834 3.6062 -0.2572>, rot=<0 0 0 -1>}
 deform_elbow_in_l {parent=deform_elbow_in_base_l, loc=<0 0 0>, rot=<0 0 0 1>}
 deform_elbow_out_base_l {parent=lowerarm_l, loc=<0 -4 0>, rot=<0 0 0 1>}
 deform_elbow_out_l {parent=deform_elbow_out_base_l, loc=<0 0 0>, rot=<0 0 0 1>}
 base_armTassels_01_l {parent=lowerarm_l, loc=<8.3252 3.9746 0.3248>, rot=<0.747 0.2224 -0.0212 -0.6262>}
 dyn_armTassels_01_l {parent=base_armTassels_01_l, loc=<0 0 0>, rot=<0 0 0 1>}
 base_armTassels_02_l {parent=lowerarm_l, loc=<4.5076 0.324 4.633>, rot=<0.0522 0.1588 0.1085 -0.9799>}
 dyn_armTassels_02_l {parent=base_armTassels_02_l, loc=<0 0 0>, rot=<0 0 0 1>}
 base_armTassels_03_l {parent=lowerarm_l, loc=<4.8472 -4.8299 0.0279>, rot=<-0.7274 0.2013 0.0566 -0.6535>}
 dyn_armTassels_03_l {parent=base_armTassels_03_l, loc=<0 0 0>, rot=<0 0 0 1>}
 upperarm_twist_01_l {parent=upperarm_l, loc=<7.2135 -0.1134 0>, rot=<0 0 0 1>}
 deform_upperarm_bk_base_l {parent=upperarm_twist_01_l, loc=<-6.4053 -4.3746 0.314>, rot=<0 0 0 -1>}
 deform_upperarm_bk_l {parent=deform_upperarm_bk_base_l, loc=<0 0 0>, rot=<0 0 0 1>}
 deform_upperarm_fr_base_l {parent=upperarm_twist_01_l, loc=<-5.6312 5.092 0.3354>, rot=<0 0 0 -1>}
 deform_upperarm_fr_l {parent=deform_upperarm_fr_base_l, loc=<0 0 0>, rot=<0 0 0 1>}
 deform_upperarm_up_base_l {parent=upperarm_twist_01_l, loc=<-8.891 0.4825 3.9517>, rot=<0 0 0 -1>}
 deform_upperarm_up_l {parent=deform_upperarm_up_base_l, loc=<0 0 0>, rot=<0 0 0 1>}
 upperarm_twist_02_l {parent=upperarm_l, loc=<19.7998 -0.1134 0>, rot=<0 0 0 1>}
 clavicle_r {parent=spine_05, loc=<5.4831 -0.1053 3.2946>, rot=<0.7289 0.0465 -0.6829 0.009>}
 upperarm_r {parent=clavicle_r, loc=<-9.8658 1.2136 0.4105>, rot=<0.0164 -0.3245 -0.0269 -0.9454>}
 lowerarm_r {parent=upperarm_r, loc=<-26.465 0.8345 -0.259>, rot=<0.0036 0.0161 -0.2298 -0.9731>}
 hand_r {parent=lowerarm_r, loc=<-22.7355 0.3203 0.6106>, rot=<-0.6784 -0.0223 0.0735 -0.7307>}
 thumb_01_r {parent=hand_r, loc=<-1.3118 0.3191 2.459>, rot=<0.6546 -0.4042 0.0664 -0.6354>}
 thumb_02_r {parent=thumb_01_r, loc=<-5.9336 0.0448 -0.0401>, rot=<-0.0037 0.0213 0.298 -0.9543>}
 thumb_03_r {parent=thumb_02_r, loc=<-2.2223 -0.3682 0.0907>, rot=<0.0118 -0.013 0.0066 -0.9998>}
 index_metacarpal_r {parent=hand_r, loc=<-1.6535 -0.0183 2.96>, rot=<0.0158 -0.0124 -0.0173 -0.9996>}
 index_01_r {parent=index_metacarpal_r, loc=<-8.3008 -0.0357 0.0524>, rot=<-0.0044 0.0485 0.0924 -0.9945>}
 index_02_r {parent=index_01_r, loc=<-3.3587 0.3004 -0.0452>, rot=<-0.002 0.0027 0.3225 -0.9466>}
 index_03_r {parent=index_02_r, loc=<-2.5578 0.123 0.0252>, rot=<-0.002 0.0067 0.2558 -0.9667>}
 middle_metacarpal_r {parent=hand_r, loc=<-2.1856 0.1725 0.5984>, rot=<0.006 0.0168 -0.0329 -0.9993>}
 middle_01_r {parent=middle_metacarpal_r, loc=<-7.3275 0.2447 -0.0129>, rot=<-0.0208 0.0042 0.0858 -0.9961>}
 middle_02_r {parent=middle_01_r, loc=<-4.0123 -0.1401 -0.1175>, rot=<-0.0218 0.0258 0.3512 -0.9357>}
 middle_03_r {parent=middle_02_r, loc=<-2.6077 -0.0534 -0.1373>, rot=<0.0082 0.0188 0.1708 -0.9851>}
 ring_metacarpal_r {parent=hand_r, loc=<-2.2074 0.3625 -1.3638>, rot=<-0.0284 0.0722 -0.0458 -0.9959>}
 ring_01_r {parent=ring_metacarpal_r, loc=<-7.0295 0.6197 0.1446>, rot=<-0.0489 -0.0252 0.1892 -0.9804>}
 ring_02_r {parent=ring_01_r, loc=<-3.6174 -0.4647 0.1027>, rot=<0 0 0.2806 -0.9598>}
 ring_03_r {parent=ring_02_r, loc=<-2.4152 -0.114 -0.0199>, rot=<0 0 0.1189 -0.9929>}
 pinky_metacarpal_r {parent=hand_r, loc=<-1.7616 0.9753 -3.3814>, rot=<-0.0397 0.0814 -0.069 -0.9935>}
 pinky_01_r {parent=pinky_metacarpal_r, loc=<-7.1709 0.4831 0.1094>, rot=<-0.0654 -0.0035 0.2242 -0.9723>}
 pinky_02_r {parent=pinky_01_r, loc=<-2.8274 -0.0893 0.1043>, rot=<0.0037 0.0011 0.2854 -0.9584>}
 pinky_03_r {parent=pinky_02_r, loc=<-2.2046 -0.0307 0.0078>, rot=<-0.0002 0.0098 0.1445 -0.9895>}
 deform_wrist_dn_base_r {parent=hand_r, loc=<-0.0001 1.5328 0>, rot=<0 0 0 -1>}
 deform_wrist_dn_r {parent=deform_wrist_dn_base_r, loc=<0 0 0>, rot=<0 0 0 1>}
 deform_wrist_up_base_r {parent=hand_r, loc=<-0.0002 -2.3007 -0.0001>, rot=<0 0 0 -1>}
 deform_wrist_up_r {parent=deform_wrist_up_base_r, loc=<0 0 0>, rot=<0 0 0 1>}
 base_handTassels_01_r {parent=hand_r, loc=<-2.3697 -1.3122 2.4744>, rot=<0.7627 -0.0266 -0.1807 -0.6204>}
 dyn_handTassels_01_r {parent=base_handTassels_01_r, loc=<0 0 0>, rot=<0 0 0 -1>}
 base_handTassels_02_r {parent=hand_r, loc=<-4.8796 -2.3763 -0.0941>, rot=<0.7037 0.0358 -0.02 -0.7093>}
 dyn_handTassels_02_r {parent=base_handTassels_02_r, loc=<0 0 0>, rot=<0 0 0 -1>}
 base_handTassels_03_r {parent=hand_r, loc=<-2.0741 -1.5488 -2.7522>, rot=<0.642 0.191 0.0062 -0.7425>}
 dyn_handTassels_03_r {parent=base_handTassels_03_r, loc=<0 0 0>, rot=<0 0 0 1>}
 lowerarm_twist_02_r {parent=lowerarm_r, loc=<-10.6707 0 0>, rot=<0 0 0 -1>}
 lowerarm_twist_01_r {parent=lowerarm_r, loc=<-18.316 0 0>, rot=<0 0 0 -1>}
 deform_elbow_in_base_r {parent=lowerarm_r, loc=<-0.1834 -3.6062 0.2572>, rot=<0 0 0 -1>}
 deform_elbow_in_r {parent=deform_elbow_in_base_r, loc=<0 0 0>, rot=<0 0 0 1>}
 deform_elbow_out_base_r {parent=lowerarm_r, loc=<0 4 -0.0001>, rot=<0 0 0 -1>}
 deform_elbow_out_r {parent=deform_elbow_out_base_r, loc=<0 0 0>, rot=<0 0 0 1>}
 base_armTassels_01_r {parent=lowerarm_r, loc=<-8.3251 -3.9747 -0.3251>, rot=<0.0212 -0.6262 0.747 -0.2224>}
 dyn_armTassels_01_r {parent=base_armTassels_01_r, loc=<0 0 0>, rot=<0 0 0 1>}
 base_armTassels_02_r {parent=lowerarm_r, loc=<-4.5076 -0.324 -4.633>, rot=<-0.1085 -0.9799 0.0522 -0.1588>}
 dyn_armTassels_02_r {parent=base_armTassels_02_r, loc=<0 0 0>, rot=<0 0 0 -1>}
 base_armTassels_03_r {parent=lowerarm_r, loc=<-4.8473 4.8299 -0.0279>, rot=<0.0566 0.6535 0.7274 0.2013>}
 dyn_armTassels_03_r {parent=base_armTassels_03_r, loc=<0 0 0>, rot=<0 0 0 1>}
 upperarm_twist_01_r {parent=upperarm_r, loc=<-7.2135 0.1134 0>, rot=<0 0 0 -1>}
 deform_upperarm_bk_base_r {parent=upperarm_twist_01_r, loc=<6.4053 4.3746 -0.314>, rot=<0 0 0 -1>}
 deform_upperarm_bk_r {parent=deform_upperarm_bk_base_r, loc=<0 0 0>, rot=<0 0 0 1>}
 deform_upperarm_fr_base_r {parent=upperarm_twist_01_r, loc=<5.6312 -5.092 -0.3354>, rot=<0 0 0 -1>}
 deform_upperarm_fr_r {parent=deform_upperarm_fr_base_r, loc=<0 0 0>, rot=<0 0 0 1>}
 deform_upperarm_up_base_r {parent=upperarm_twist_01_r, loc=<8.891 -0.4825 -3.9517>, rot=<0 0 0 -1>}
 deform_upperarm_up_r {parent=deform_upperarm_up_base_r, loc=<0 0 0>, rot=<0 0 0 1>}
 upperarm_twist_02_r {parent=upperarm_r, loc=<-19.7998 0.1134 0>, rot=<0 0 0 -1>}
 attach_backpack {parent=spine_05, loc=<-0.123 -13.473 0>, rot=<-0.0093 -0.707 0.0093 -0.707>}
 Backpack_BR {parent=attach_backpack, type=Null, loc=<0 -4.6009 0>, rot=<0 0 0 1>}
 attach_cape {parent=spine_05, loc=<9.017 -11.714 0>, rot=<0.0616 -0.7044 -0.0616 -0.7044>}
 cape {parent=attach_cape, type=Null, loc=<0 -3.5129 -2.0985>, rot=<0.023 0 0 0.9997>}
 deform_pec_aim_base_l {parent=spine_04, loc=<3.2456 9.3309 -1.2237>, rot=<0.0603 -0.4033 0.3318 -0.8507>}
 deform_pec_aim_l {parent=deform_pec_aim_base_l, loc=<0 0 0>, rot=<0 0 0 1>}
 deform_pec_base_l {parent=deform_pec_aim_l, loc=<9.3665 0 0>, rot=<0 0 0 1>}
 deform_pec_l {parent=deform_pec_base_l, loc=<0 0 0>, rot=<0 0 0 1>}
 deform_pec_aim_base_r {parent=spine_04, loc=<3.2452 9.3309 1.2237>, rot=<-0.4033 -0.0603 0.8507 0.3318>}
 deform_pec_aim_r {parent=deform_pec_aim_base_r, loc=<0 0 0>, rot=<0 0 0 1>}
 deform_pec_base_r {parent=deform_pec_aim_r, loc=<-9.3667 -0.0003 -0.0004>, rot=<0 0 0 1>}
 deform_pec_r {parent=deform_pec_base_r, loc=<0 0 0>, rot=<0 0 0 1>}
 base_torsoTassels_02_l {parent=spine_03, loc=<2.2605 9.5442 -7.9695>, rot=<0.0466 -0.8499 -0.5241 0.0287>}
 dyn_torsoTassels_02_l {parent=base_torsoTassels_02_l, loc=<0 0 0>, rot=<0 0 0 1>}
 base_torsoTassels_03_l {parent=spine_03, loc=<4.538 7.4689 -10.4854>, rot=<0.0466 -0.8499 -0.5241 0.0287>}
 dyn_torsoTassels_03_l {parent=base_torsoTassels_03_l, loc=<0 0 0>, rot=<0 0 0 1>}
 base_torsoTassels_01_l {parent=spine_03, loc=<2.3682 11.6496 -4.0894>, rot=<0.0428 -0.7809 -0.6222 0.0341>}
 dyn_torsoTassels_01_l {parent=base_torsoTassels_01_l, loc=<0 0 0>, rot=<0 0 0 1>}
 base_torsoTassels_01_m {parent=spine_03, loc=<5.674 12.3643 0>, rot=<-0.0387 0.706 0.706 -0.0387>}
 dyn_torsoTassels_01_m {parent=base_torsoTassels_01_m, loc=<0 0 0>, rot=<0 0 0 1>}
 base_torsoTassels_01_r {parent=spine_03, loc=<2.368 11.6495 4.0894>, rot=<-0.0341 0.6222 0.7809 -0.0428>}
 dyn_torsoTassels_01_r {parent=base_torsoTassels_01_r, loc=<0 0 0>, rot=<0 0 0 1>}
 base_torsoTassels_02_r {parent=spine_03, loc=<2.2605 9.5442 7.9694>, rot=<-0.0287 0.5241 0.8499 -0.0466>}
 dyn_torsoTassels_02_r {parent=base_torsoTassels_02_r, loc=<0 0 0>, rot=<0 0 0 1>}
 base_torsoTassels_03_r {parent=spine_03, loc=<4.538 7.4689 10.4854>, rot=<-0.0287 0.5241 0.8499 -0.0466>}
 dyn_torsoTassels_03_r {parent=base_torsoTassels_03_r, loc=<0 0 0>, rot=<0 0 0 1>}
 deform_lat_l_base {parent=spine_03, loc=<1.4471 -4.8405 -13.6119>, rot=<0 0 0 1>}
 deform_lat_l {parent=deform_lat_l_base, loc=<0 0 0>, rot=<0 0 0 1>}
 deform_lat_r_base {parent=spine_03, loc=<1.447 -4.8405 13.6119>, rot=<0 0 1 0>}
 deform_lat_r {parent=deform_lat_r_base, loc=<-0.0001 0 0>, rot=<0 0 0 -1>}
 thigh_l {parent=pelvis, loc=<-1.0855 -0.2109 -8.7053>, rot=<0.0009 0.0355 0.0155 -0.9993>}
 calf_l {parent=thigh_l, loc=<-42.5284 -0.0019 -0.7297>, rot=<0.0006 -0.0161 -0.0391 -0.9991>}
 foot_l {parent=calf_l, loc=<-46.7401 -2.5014 -0.592>, rot=<0.0427 -0.0238 0.0353 -0.9982>}
 ball_l {parent=foot_l, loc=<-3.5658 10.8066 -0.1544>, rot=<0.0001 -0.0001 -0.7063 -0.7079>}
 calf_twist_02_l {parent=calf_l, loc=<-19.1546 -1.7943 -0.2638>, rot=<0.0237 0.0095 -0.0035 -0.9997>}
 calf_twist_01_l {parent=calf_l, loc=<-35.8483 -1.9197 -0.5794>, rot=<0.0237 0.0095 -0.0035 -0.9997>}
 deform_knee_in_base_l {parent=calf_l, loc=<0.8031 -5.0231 0.2456>, rot=<0 0 0 -1>}
 deform_knee_in_l {parent=deform_knee_in_base_l, loc=<0 0 0>, rot=<0 0 0 1>}
 deform_knee_out_base_l {parent=calf_l, loc=<0 4 0>, rot=<0 0 0 -1>}
 deform_knee_out_l {parent=deform_knee_out_base_l, loc=<0 0 0>, rot=<0 0 0 1>}
 thigh_twist_01_l {parent=thigh_l, loc=<-27.952 0 0>, rot=<0.0013 0 0 -1>}
 deform_groin_base_l {parent=thigh_l, loc=<0.4801 5.6776 -0.2486>, rot=<0 0 0 -1>}
 deform_groin_l {parent=deform_groin_base_l, loc=<0 0 0>, rot=<0 0 0 1>}
 base_tasselLeg_01_l {parent=thigh_l, loc=<-4.3154 6.3374 3.5625>, rot=<0.0578 0.535 0.8349 0.1156>}
 dyn_tasselLeg_01_l {parent=base_tasselLeg_01_l, loc=<0 0 0>, rot=<0 0 0 1>}
 base_tasselLeg_02_l {parent=thigh_l, loc=<-2.1755 7.0511 0.4267>, rot=<0.0119 0.629 0.7718 0.092>}
 dyn_tasselLeg_02_l {parent=base_tasselLeg_02_l, loc=<0 0 0>, rot=<0 0 0 1>}
 base_tasselLeg_03_l {parent=thigh_l, loc=<-0.6185 6.6465 -2.6076>, rot=<0.0269 -0.8472 -0.523 -0.0892>}
 dyn_tasselLeg_03_l {parent=base_tasselLeg_03_l, loc=<0 0 0>, rot=<0 0 0 1>}
 base_tasselLeg_04_l {parent=thigh_l, loc=<0.0159 4.374 -5.1195>, rot=<-0.0392 -0.9317 -0.3202 -0.167>}
 dyn_tasselLeg_04_l {parent=base_tasselLeg_04_l, loc=<0 0 0>, rot=<0 0 0 1>}
 base_tasselLeg_05_l {parent=thigh_l, loc=<-0.3758 0.8926 -6.3834>, rot=<-0.0469 -0.9737 -0.0965 -0.2011>}
 dyn_tasselLeg_05_l {parent=base_tasselLeg_05_l, loc=<0 0 0>, rot=<0 0 0 1>}
 base_tasselLeg_06_l {parent=thigh_l, loc=<-1.5846 -3.2508 -6.4174>, rot=<-0.0565 -0.9771 0.0844 -0.1869>}
 dyn_tasselLeg_06_l {parent=base_tasselLeg_06_l, loc=<0 0 0>, rot=<0 0 0 1>}
 base_tasselLeg_07_l {parent=thigh_l, loc=<-3.165 -7.1721 -5.7316>, rot=<-0.022 -0.9482 0.2359 -0.2114>}
 dyn_tasselLeg_07_l {parent=base_tasselLeg_07_l, loc=<0 0 0>, rot=<0 0 0 1>}
 base_tasselLeg_08_l {parent=thigh_l, loc=<-5.5427 -10.6766 -3.7877>, rot=<-0.0303 -0.906 0.3961 -0.1463>}
 dyn_tasselLeg_08_l {parent=base_tasselLeg_08_l, loc=<0 0 0>, rot=<0 0 0 1>}
 base_tasselLeg_09_l {parent=thigh_l, loc=<-8.5797 -12.9122 -0.3186>, rot=<-0.0845 -0.7438 0.6598 -0.0659>}
 dyn_tasselLeg_09_l {parent=base_tasselLeg_09_l, loc=<0 0 0>, rot=<0 0 0 1>}
 base_tasselLeg_10_l {parent=thigh_l, loc=<-10.963 -11.723 3.7195>, rot=<0.008 -0.5067 0.862 -0.0164>}
 dyn_tasselLeg_10_l {parent=base_tasselLeg_10_l, loc=<0 0 0>, rot=<0 0 0 1>}
 thigh_r {parent=pelvis, loc=<-1.0855 -0.2109 8.7053>, rot=<0.0355 -0.0009 0.9993 0.0155>}
 calf_r {parent=thigh_r, loc=<42.5284 0.0019 0.7297>, rot=<0.0006 -0.0161 -0.0391 -0.9991>}
 foot_r {parent=calf_r, loc=<46.7401 2.5014 0.592>, rot=<0.0427 -0.0238 0.0353 -0.9982>}
 ball_r {parent=foot_r, loc=<3.5658 -10.8066 0.1544>, rot=<0.0001 -0.0001 -0.7063 -0.7079>}
 calf_twist_02_r {parent=calf_r, loc=<19.1546 1.7943 0.2638>, rot=<0.0237 0.0095 -0.0035 -0.9997>}
 calf_twist_01_r {parent=calf_r, loc=<35.8483 1.9197 0.5794>, rot=<0.0237 0.0095 -0.0035 -0.9997>}
 deform_knee_in_base_r {parent=calf_r, loc=<-0.8031 5.0231 -0.2457>, rot=<0 0 0 -1>}
 deform_knee_in_r {parent=deform_knee_in_base_r, loc=<0 0 0>, rot=<0 0 0 1>}
 deform_knee_out_base_r {parent=calf_r, loc=<0 -4 0>, rot=<0 0 0 -1>}
 deform_knee_out_r {parent=deform_knee_out_base_r, loc=<0 0 0>, rot=<0 0 0 1>}
 thigh_twist_01_r {parent=thigh_r, loc=<27.952 0 0>, rot=<0.0013 0 0 -1>}
 deform_groin_base_r {parent=thigh_r, loc=<-0.4801 -5.6776 0.2486>, rot=<0 0 0 -1>}
 deform_groin_r {parent=deform_groin_base_r, loc=<0 0 0>, rot=<0 0 0 1>}
 base_tasselLeg_01_r {parent=thigh_r, loc=<4.3154 -6.3374 -3.5624>, rot=<-0.8349 0.1156 0.0578 -0.535>}
 dyn_tasselLeg_01_r {parent=base_tasselLeg_01_r, loc=<0 0 0>, rot=<0 0 0 1>}
 base_tasselLeg_02_r {parent=thigh_r, loc=<2.1754 -7.0511 -0.4267>, rot=<-0.7718 0.092 0.0119 -0.629>}
 dyn_tasselLeg_02_r {parent=base_tasselLeg_02_r, loc=<0 0 0>, rot=<0 0 0 1>}
 base_tasselLeg_03_r {parent=thigh_r, loc=<0.6185 -6.6465 2.6076>, rot=<-0.523 0.0892 -0.0269 -0.8472>}
 dyn_tasselLeg_03_r {parent=base_tasselLeg_03_r, loc=<0 0 0>, rot=<0 0 0 1>}
 base_tasselLeg_04_r {parent=thigh_r, loc=<-0.0159 -4.374 5.1196>, rot=<-0.3202 0.167 0.0392 -0.9317>}
 dyn_tasselLeg_04_r {parent=base_tasselLeg_04_r, loc=<0 0 0>, rot=<0 0 0 1>}
 base_tasselLeg_05_r {parent=thigh_r, loc=<0.3758 -0.8926 6.3834>, rot=<-0.0965 0.2011 0.0469 -0.9737>}
 dyn_tasselLeg_05_r {parent=base_tasselLeg_05_r, loc=<0 0 0>, rot=<0 0 0 1>}
 base_tasselLeg_06_r {parent=thigh_r, loc=<1.5846 3.2508 6.4174>, rot=<0.0844 0.1869 0.0565 -0.9771>}
 dyn_tasselLeg_06_r {parent=base_tasselLeg_06_r, loc=<0 0 0>, rot=<0 0 0 1>}
 base_tasselLeg_07_r {parent=thigh_r, loc=<3.1649 7.172 5.7316>, rot=<0.2359 0.2114 0.022 -0.9482>}
 dyn_tasselLeg_07_r {parent=base_tasselLeg_07_r, loc=<0 0 0>, rot=<0 0 0 1>}
 base_tasselLeg_08_r {parent=thigh_r, loc=<5.5426 10.6766 3.7877>, rot=<0.3961 0.1463 0.0303 -0.906>}
 dyn_tasselLeg_08_r {parent=base_tasselLeg_08_r, loc=<0 0 0>, rot=<0 0 0 1>}
 base_tasselLeg_09_r {parent=thigh_r, loc=<8.5797 12.9122 0.3186>, rot=<0.6598 0.0659 0.0845 -0.7438>}
 dyn_tasselLeg_09_r {parent=base_tasselLeg_09_r, loc=<0 0 0>, rot=<0 0 0 1>}
 base_tasselLeg_10_r {parent=thigh_r, loc=<10.963 11.723 -3.7195>, rot=<0.862 0.0164 -0.008 -0.5067>}
 dyn_tasselLeg_10_r {parent=base_tasselLeg_10_r, loc=<0 0 0>, rot=<0 0 0 1>}
 deform_glute_base_l {parent=pelvis, loc=<-7.3466 -9.0213 -9.1245>, rot=<0.0172 0.031 -0.4477 -0.8935>}
 deform_glute_l {parent=deform_glute_base_l, loc=<0 0 0>, rot=<0 0 0 1>}
 deform_glute_base_r {parent=pelvis, loc=<-7.3464 -9.0213 9.1245>, rot=<0.031 -0.0172 0.8935 -0.4477>}
 deform_glute_r {parent=deform_glute_base_r, loc=<0 0 0>, rot=<0 0 0 1>}
 thigh_l_null {parent=pelvis, type=Null, loc=<-1.0855 -0.2109 -8.7053>, rot=<0.0009 0.0355 0.0155 -0.9993>}
 thigh_r_null {parent=pelvis, type=Null, loc=<-1.0855 -0.2109 8.7053>, rot=<0.0355 -0.0009 0.9993 0.0155>}
]
